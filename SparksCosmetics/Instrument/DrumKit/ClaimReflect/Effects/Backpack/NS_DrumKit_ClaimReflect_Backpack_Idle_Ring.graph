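# Niagara idle effect for the ClaimReflect drum kit back bling, showing a halo ring and UNSC hologram with a blue glow.
# HaloRingAndUNSCHologram burst-spawns a persistent particle drawn by five mesh renderers and a light renderer, alongside Hologram_Glow's soft, camera-offset sprite.
# Time-of-day scalars adjust both emitters' brightness.
# User bools shrink and fade the meshes during ADS and change a mesh material parameter in front-end preview.
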NiagaraSystem /SparksCosmetics/Instrument/DrumKit/ClaimReflect/Effects/Backpack/NS_DrumKit_ClaimReflect_Backpack_Idle_Ring
dependencies {usedDataInterfaces [&NiagaraDataInterfaceCurve], usedDynamicInputs [], usedModules [&SystemState]}
emitters {
 HaloRingAndUNSCHologram {
  emitterUpdateScript [
   &EmitterState(
    Inactive Response=::NewEnumerator1
    Life Cycle Mode=::NewEnumerator1
    Loop Behavior=::NewEnumerator1
    Loop Duration Mode=::NewEnumerator1
    UseLoopDelay=false
   )
   &SpawnBurst_Instantaneous(Use Spawn Probability=false, UseLoopCountLimit=false)
  ]
  particleSpawnScript [
   &InitializeParticle(
    Color=<rgba 0.5156 0.5156 0.5156 0.5043>
    Mesh Scale Mode=::NewEnumerator4
    Position Offset=<2 23.5 0>
    UsePositionOffset=true
    Write Material Random=true
   )
   /SetVariables(Particles.LightColor=<rgba 0.0026 1.7507 5 1>) {isSetParameters=true}
   &InitialMeshOrientation(Mesh Orientation Mode=::NewEnumerator2, Rotation=<0.5 0 0.5>, Use Rotation Vector=true)
  ]
  particleUpdateScript [
   /SetVariables(Particles.VisibilityTag=&SpawnGroup()) {enabled=false, isSetParameters=true}
   &ParticleState(Kill Particles When Lifetime Has Elapsed=false, Let Infinitely Lived Particles Die When Emitter Deactivates=true)
   &ScaleMeshSize(Scale Factor=&VectorFromFloat(Value=&MakeCustomFloatFromBool(Bool=$User.bIsADS, True Float=0.9, False Float=1)))
   &UpdateMeshOrientation(Rotation Coordinate Space=::NewEnumerator2, Rotation Rate=0.05, Rotation Vector=<0 0.8 0>) {enabled=false}
   &SolveForcesAndVelocity(Clamp Velocity=false, Limit Acceleration=false, Speed Limit=200) {enabled=false}
   &ScaleColor(
    Scale RGB=&Multiply_VectorByFloat(
     Vector=<1 1 1>
     Float=&FNTimeOfDayScalar(Night Scale=3, Morning Scale=2, Day Scale=2, Evening Scale=2)
    )
    ScaleA=true
    ScaleRGB=true
   )
   &ScaleColor(
    Scale Alpha=&MakeCustomFloatFromBool(Bool=$User.bIsADS, True Float=0.75, False Float=1)
    ScaleA=true
    ScaleRGB=false
   ) {module=ScaleColor001}
   &DynamicMaterialParameters(
    Index 0 Param 2=&MakeCustomFloatFromBool(Bool=$User.bIsFrontEndPreview, True Float=0.3, False Float=0.075)
    Module.Index0Float0Write=true
    Module.Index0Float1Write=true
    Module.Index0Float2Write=true
    Module.Index0Float3Write=true
   )
  ]
  rendererClasses [&NiagaraMeshRendererProperties, &NiagaraLightRendererProperties]
  renderers [
   {rendererClass=&NiagaraMeshRendererProperties, rendererIndex=0}
   {rendererClass=&NiagaraMeshRendererProperties, rendererIndex=1}
   {rendererClass=&NiagaraMeshRendererProperties, rendererIndex=2}
   {rendererClass=&NiagaraMeshRendererProperties, rendererIndex=3}
   {rendererClass=&NiagaraMeshRendererProperties, rendererIndex=4}
   {rendererClass=&NiagaraLightRendererProperties, rendererIndex=5}
  ]
 }
 Hologram_Glow {
  emitterUpdateScript [
   &EmitterState(Life Cycle Mode=::NewEnumerator1, Loop Behavior=::NewEnumerator1, Loop Duration Mode=::NewEnumerator1, UseLoopDelay=false)
   &SpawnBurst_Instantaneous(Use Spawn Probability=false, UseLoopCountLimit=false)
  ]
  particleSpawnScript [
   &PartitionParticles()
   &InitializeParticle(
    Color=<rgba 0 0.8332 2 0.3739>
    Position Offset=<2 23.5 0>
    Sprite Size Mode=::NewEnumerator3
    Uniform Sprite Size=55
    UsePositionOffset=true
    Write Material Random=true
   ) {module=InitializeParticle001}
  ]
  particleUpdateScript [
   &ParticleState(
    Kill Particles When Lifetime Has Elapsed=false
    Let Infinitely Lived Particles Die When Emitter Deactivates=true
    Loop Particles Lifetime=true
   )
   &DynamicMaterialParameters(
    Index 0 Param 1=4
    Index 0 Param 2=25
    Index 0 Param 3=4
    Index 0 Param 4=3
    Module.Index0Param0Write=true
    Module.Index0Param1Write=true
    Module.Index0Param2Write=true
    Module.Index0Param3Write=true
   )
   &FN_Light_Attributes(
    Color=$Particles.Color
    Create Color Attribute=true
    Create Exponent Attribute=true
    Create Radius Attribute=true
    Enable Particle Light Enable=true
    Light Enabled=true
    Radius=100
    Scale Color=&Multiply_Float(A=50, B=&MakeFloatFromVector(Vector=$Engine.Owner.Scale, Channel=::NewEnumerator0))
   )
   &ScaleColor(
    Scale Alpha=&MakeFloatFromVector(
     Vector=&VectorFromFloat(Value=&Lerp_Float(
       A=1.25
       B=1
       Alpha=&MakeCustomFloatFromBool(Bool=$User.bIsFrontEndPreview, True Float=1, False Float=0)
      ))
     Channel=::NewEnumerator0
    )
    Scale RGB=&VectorFromFloat(Value=&Lerp_Float(
      A=0.8
      B=0.75
      Alpha=&MakeCustomFloatFromBool(Bool=$User.bIsFrontEndPreview, True Float=1, False Float=0)
     ))
    ScaleA=true
    ScaleRGB=true
   ) {enabled=false, module=ScaleColor001}
   &CameraOffset(Camera Offset Amount=-1)
   &ScaleColor(Scale Alpha=0.15, ScaleA=true, ScaleRGB=true)
   &ScaleColor(
    Scale RGB=&Multiply_VectorByFloat(
     Vector=<1 1 1>
     Float=&FNTimeOfDayScalar(Night Scale=0.6, Morning Scale=0.9, Day Scale=0.8, Evening Scale=0.9)
    )
    ScaleA=true
    ScaleRGB=true
   ) {module=ScaleColor002}
  ]
  rendererClasses [&NiagaraSpriteRendererProperties]
  renderers [{rendererClass=&NiagaraSpriteRendererProperties}]
 }
}
summary {
 systemName=NS_DrumKit_ClaimReflect_Backpack_Idle_Ring
 emitters [
  {emitterName=Hologram_Glow, rendererClasses [&NiagaraSpriteRendererProperties]}
  {
   emitterName=HaloRingAndUNSCHologram
   rendererClasses [&NiagaraMeshRendererProperties, &NiagaraLightRendererProperties]
  }
 ]
 userVariables [
  User.bIsADS {type=NiagaraBool}
  User.bIsFrontEndPreview {type=NiagaraBool}
 ]
}
